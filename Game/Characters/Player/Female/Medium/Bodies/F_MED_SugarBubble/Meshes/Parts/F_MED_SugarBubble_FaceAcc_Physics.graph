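# Physics asset for the F_MED_SugarBubble female player face accessory mesh, defining simulation bodies for its dynamic hair bones.
# Five capsule chains (dyn_hair_a to e) hang from a small head body, each two bones long except e, which has three.
# Tight swing and twist limits keep the strands close to their posed shape.
# The head joins spine_05 with free angular motion, and spine_05 carries two large capsules covering the upper torso.
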PhysicsAsset /Game/Characters/Player/Female/Medium/Bodies/F_MED_SugarBubble/Meshes/Parts/F_MED_SugarBubble_FaceAcc_Physics
bodies {
 dyn_hair_a_01 [{
   shapeType=Capsule
   length=2.7019
   radius=1
   shapeName=dyn_hair_a_01_capsule
   center=<2.1281 -0.0686 -0.0958>
   rotation=<rot -10.9971 87.025 -82.5882>
  }]
 dyn_hair_a_02 [{
   shapeType=Capsule
   length=1.6154
   radius=1
   shapeName=dyn_hair_a_02_capsule
   center=<2.1042 0.5558 0.0883>
   rotation=<rot 0 112.938 -86.8625>
  }]
 dyn_hair_b_01 [{
   shapeType=Capsule
   length=1.1145
   radius=1
   shapeName=dyn_hair_b_01_capsule
   center=<3.1831 0.0726 0.0604>
   rotation=<rot 0 -170.0555 -77.4635>
  }]
 dyn_hair_b_02 [{
   shapeType=Capsule
   length=2.2651
   radius=1
   shapeName=dyn_hair_b_02_capsule
   center=<1.8324 -0.922 -1.29>
   rotation=<rot 0 -130.7983 -55.1148>
  }]
 dyn_hair_c_01 [{shapeType=Capsule, length=0.505, radius=1, shapeName=dyn_hair_c_01_capsule, rotation=<rot 0 -90 0>}]
 dyn_hair_c_02 [{
   shapeType=Capsule
   length=1.2893
   radius=1
   shapeName=dyn_hair_c_02_capsule
   center=<1.0617 -0.5056 -0.4607>
   rotation=<rot 0 -123.1618 -55.9137>
  }]
 dyn_hair_d_01 [{
   shapeType=Capsule
   length=0.505
   radius=1
   shapeName=dyn_hair_d_01_capsule
   center=<2.1233 0 0.0342>
   rotation=<rot 89.0774 -179.997 -179.9971>
  }]
 dyn_hair_d_02 [{
   shapeType=Capsule
   length=3.0574
   radius=1
   shapeName=dyn_hair_d_02_capsule
   center=<1.6478 -0.5022 -0.6881>
   rotation=<rot 0 -83.7177 -67.0933>
  }]
 dyn_hair_e_01 [{
   shapeType=Capsule
   length=3.0093
   radius=1
   shapeName=dyn_hair_e_01_capsule
   center=<1.8072 0.6215 0.2938>
   rotation=<rot 10.2113 -90.5667 -80.3121>
  }]
 dyn_hair_e_02 [{shapeType=Capsule, length=2.505, radius=1, shapeName=dyn_hair_e_02_capsule, center=<2.2507 0 -0.0219>, rotation=<rot 0 -90 -83.9112>}]
 dyn_hair_e_03 [{
   shapeType=Capsule
   length=2.655
   radius=1
   shapeName=dyn_hair_e_03_capsule
   center=<3.6467 -0.6526 -1.0879>
   rotation=<rot 0 -97.8176 -69.5974>
  }]
 head [{shapeType=Capsule, length=0, radius=0.5, center=<2.5178 0.3914 0.8238>, rotation=<rot 0 136.7425 -89.7595>}]
 spine_05 [
  {
   shapeType=Capsule
   length=19.5
   radius=10.5
   shapeName=spine_05_capsule
   center=<15.2867 0.8642 -17.3624>
   rotation=<rot 66.1133 89.9999 -180>
  }
  {shapeType=Capsule, length=18.5, radius=11.5, shapeName=spine_05_capsule, center=<13.4088 -0.2832 18.3852>, rotation=<rot 0 -0.0001 81.7678>}
 ]
}
constraints [
 {bone1Name=dyn_hair_e_03, bone2Name=dyn_hair_e_02, swing1LimitDegrees=10, swing2LimitDegrees=5, twistLimitDegrees=5}
 {bone1Name=dyn_hair_a_01, bone2Name=head, swing1LimitDegrees=5, swing2LimitDegrees=5, twistLimitDegrees=5}
 {bone1Name=dyn_hair_a_02, bone2Name=dyn_hair_a_01, swing1LimitDegrees=5, swing2LimitDegrees=5, twistLimitDegrees=5}
 {bone1Name=dyn_hair_b_01, bone2Name=head, swing1LimitDegrees=10, swing2LimitDegrees=10, twistLimitDegrees=5}
 {bone1Name=dyn_hair_b_02, bone2Name=dyn_hair_b_01, swing1LimitDegrees=10, swing2LimitDegrees=10, twistLimitDegrees=5}
 {bone1Name=dyn_hair_c_01, bone2Name=head, swing1LimitDegrees=7, swing2LimitDegrees=7, twistLimitDegrees=5}
 {bone1Name=dyn_hair_c_02, bone2Name=dyn_hair_c_01, swing1LimitDegrees=7, swing2LimitDegrees=7, twistLimitDegrees=5}
 {bone1Name=dyn_hair_d_01, bone2Name=head, swing1LimitDegrees=5, swing2LimitDegrees=5, twistLimitDegrees=5}
 {bone1Name=dyn_hair_d_02, bone2Name=dyn_hair_d_01, swing1LimitDegrees=5, swing2LimitDegrees=5, twistLimitDegrees=5}
 {bone1Name=dyn_hair_e_01, bone2Name=head, swing1LimitDegrees=5, swing2LimitDegrees=1, twistLimitDegrees=5}
 {bone1Name=dyn_hair_e_02, bone2Name=dyn_hair_e_01, swing1LimitDegrees=10, swing2LimitDegrees=3, twistLimitDegrees=5}
 {
  bone1Name=head
  bone2Name=spine_05
  swing1LimitDegrees=45
  swing1Motion=Free
  swing2LimitDegrees=45
  swing2Motion=Free
  twistLimitDegrees=45
  twistMotion=Free
 }
]
