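# Animation Blueprint for the SugarBubble female FaceAcc part that adds secondary motion to hair bones.
# It copies the attached parent's pose, adds left head-roll rotation to strand e, then runs spring-based AnimDynamics on strands a–e.
# Spherical limits driven by spine_05 constrain two strand tips.
# The RigidBody node, right-roll bone offsets and e_01 dynamics are disconnected from the output, though the event graph still updates gravity and both roll alphas.
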AnimBlueprint /BRCosmetics/Characters/Player/Female/Medium/Bodies/F_MED_SugarBubble/Meshes/Parts/F_MED_SugarBubble_FaceAcc_AnimBP
target_skeleton=&/F_MED_SugarBubble_FaceAcc_Skeleton
graphs [
 AnimGraph {nodes [
   Anim Dynamics - Bone: dyn_hair_e_01(Alpha=0.7, ExternalForce="0,0,0") {
    angularDampingOverride=1
    angularSpringConstant=500
    bAngularSpring=true
    bOverrideAngularDamping=true
    bUseGravityOverride=true
    gravityScale=0.35
    simulationSpace=RootRelative
    alphaBoolBlend {blendOption=Linear}
    boundBone {boneName=dyn_hair_e_01}
    componentLinearAccScale=<1 1 1>
    componentLinearVelScale=<1 1 1>
    gravityOverride {z=-980}
    physicsBodyDefinitions [{
      boundBone {boneName=dyn_hair_e_01}
      boxExtents=<10 10 10>
      constraintSetup {angularLimitsMax=<20 5 20>, angularLimitsMin=<-20 -30 -20>, angularTarget=<0 0 0>}
      localJointOffset=<3 0 0>
     }]
    relativeSpaceBone {boneName=head}
   }
   n1 = RigidBody!(ComponentPose=Local To Component(), Alpha, OverrideWorldGravity:Vector=Get Gravity Override(), ExternalForce) -> (Pose) {
    bClampLinearTranslationLimitToRefPose=true
    bOverrideWorldGravity=true
    overridePhysicsAsset=/Game/Characters/Player/Female/Medium/Bodies/F_MED_SugarBubble/Meshes/Parts/F_MED_SugarBubble_FaceAcc_Physics
    simulationSpace=BaseBoneSpace
    alphaBoolBlend {blendOption=Linear}
    alphaScaleBiasClamp {bInterpResult=true, interpSpeedDecreasing=5, interpSpeedIncreasing=5}
    baseBoneRef {boneName=neck_01}
    componentAppliedLinearAccClamp=<3000 3000 3000>
    componentLinearAccScale=<0.4 0.4 0.4>
    componentLinearVelScale=<0.6 0.6 0.6>
    overrideWorldGravity {z=-980}
    simSpaceSettings {
     maxAngularAcceleration=500
     maxAngularVelocity=3
     maxLinearAcceleration=400
     maxLinearVelocity=250
     velocityScaleZ=0
     worldAlpha=1
     externalLinearDragV=<0.1 0.1>
    }
   }
   Component To Local(ComponentPose=@n1)
   n2 = Get head rotation right() -> (head rotation right:double)
   n3 = "Transform (Modify) Bone - Bone: dyn_hair_a_01"(Alpha=@n2, Translation="0,0,0", Rotation="0,0,-11", Scale="1,1,1") {rotationMode=BMM_Additive, alphaBoolBlend {blendOption=Linear}, boneToModify {boneName=dyn_hair_a_01}, rotation {roll=-11}}
   "Transform (Modify) Bone - Bone: dyn_hair_a_02"(ComponentPose=@n3, Alpha=@n2, Translation="0,0,0", Rotation="0.000,0.000,-13.846", Scale="1,1,1") {rotationMode=BMM_Additive, alphaBoolBlend {blendOption=Linear}, boneToModify {boneName=dyn_hair_a_02}, rotation {roll=-13.846}}
   n4 = Get head rotation left() -> (head rotation left:double)
   n5 = "Transform (Modify) Bone - Bone: dyn_hair_e_01"(
    ComponentPose=Local To Component(LocalPose=Copy Pose From Mesh() {bUseAttachedParent=true})
    Alpha=@n4
    Translation="0,0,0"
    Rotation="0,0,10"
    Scale="1,1,1"
   ) {rotationMode=BMM_Additive, alphaBoolBlend {blendOption=Linear}, boneToModify {boneName=dyn_hair_e_01}, rotation {roll=10}}
   n6 = "Transform (Modify) Bone - Bone: dyn_hair_e_02"(ComponentPose=@n5, Alpha=@n4, Translation="0,0,0", Rotation="0,0,10", Scale="1,1,1") {rotationMode=BMM_Additive, alphaBoolBlend {blendOption=Linear}, boneToModify {boneName=dyn_hair_e_02}, rotation {roll=10}}
   n7 = Anim Dynamics - Bone: dyn_hair_d_01(
    ComponentPose="Transform (Modify) Bone - Bone: dyn_hair_e_03"(ComponentPose=@n6, Alpha=@n4, Translation="0,0,0", Rotation="0,0,0", Scale="1,1,1") {rotationMode=BMM_Additive, alphaBoolBlend {blendOption=Linear}, boneToModify {boneName=dyn_hair_e_03}}
    Alpha=0.7
    ExternalForce="0,0,0"
   ) {
    angularDampingOverride=1
    angularSpringConstant=500
    bAngularSpring=true
    bOverrideAngularDamping=true
    bUseGravityOverride=true
    gravityScale=0.35
    simulationSpace=RootRelative
    alphaBoolBlend {blendOption=Linear}
    boundBone {boneName=dyn_hair_d_01}
    componentLinearAccScale=<1 1 1>
    componentLinearVelScale=<1 1 1>
    gravityOverride {z=-980}
    physicsBodyDefinitions [{
      boundBone {boneName=dyn_hair_d_01}
      boxExtents=<10 10 10>
      constraintSetup {angularLimitsMax=<20 5 20>, angularLimitsMin=<-20 -30 -20>, angularTarget=<0 0 0>}
      localJointOffset=<3 0 0>
     }]
    relativeSpaceBone {boneName=head}
   }
   n8 = Anim Dynamics - Bone: dyn_hair_d_02(ComponentPose=@n7, Alpha=0.7, ExternalForce="0,0,0") {
    angularDampingOverride=1
    angularSpringConstant=500
    bAngularSpring=true
    bOverrideAngularDamping=true
    bUseGravityOverride=true
    gravityScale=0.35
    simulationSpace=RootRelative
    alphaBoolBlend {blendOption=Linear}
    boundBone {boneName=dyn_hair_d_02}
    componentLinearAccScale=<1 1 1>
    componentLinearVelScale=<1 1 1>
    gravityOverride {z=-980}
    physicsBodyDefinitions [{
      boundBone {boneName=dyn_hair_d_02}
      boxExtents=<10 10 10>
      constraintSetup {angularLimitsMax=<20 5 20>, angularLimitsMin=<-20 -30 -20>, angularTarget=<0 0 0>}
      localJointOffset=<3 0 0>
     }]
    relativeSpaceBone {boneName=head}
   }
   n9 = Anim Dynamics - Bone: dyn_hair_c_01(ComponentPose=@n8, Alpha=0.6, ExternalForce="0,0,0") {
    angularDampingOverride=1
    angularSpringConstant=500
    bAngularSpring=true
    bOverrideAngularDamping=true
    bUseGravityOverride=true
    gravityScale=0.35
    simulationSpace=RootRelative
    alphaBoolBlend {blendOption=Linear}
    boundBone {boneName=dyn_hair_c_01}
    componentLinearAccScale=<1 1 1>
    componentLinearVelScale=<1 1 1>
    gravityOverride {z=-980}
    physicsBodyDefinitions [{
      boundBone {boneName=dyn_hair_c_01}
      boxExtents=<10 10 10>
      constraintSetup {angularLimitsMax=<20 20 45>, angularLimitsMin=<-20 -30 -20>, angularTarget=<0 0 0>}
      localJointOffset=<3 0 0>
     }]
    relativeSpaceBone {boneName=head}
   }
   n10 = Anim Dynamics - Bone: dyn_hair_c_02(ComponentPose=@n9, Alpha=0.5, ExternalForce="0,0,0") {
    angularDampingOverride=1
    angularSpringConstant=500
    bAngularSpring=true
    bOverrideAngularDamping=true
    bUseGravityOverride=true
    gravityScale=0.35
    simulationSpace=RootRelative
    alphaBoolBlend {blendOption=Linear}
    boundBone {boneName=dyn_hair_c_02}
    componentLinearAccScale=<1 1 1>
    componentLinearVelScale=<1 1 1>
    gravityOverride {z=-980}
    physicsBodyDefinitions [{
      boundBone {boneName=dyn_hair_c_02}
      boxExtents=<10 10 10>
      constraintSetup {angularLimitsMax=<20 20 45>, angularLimitsMin=<-20 -30 -20>, angularTarget=<0 0 0>}
      localJointOffset=<3 0 0>
     }]
    relativeSpaceBone {boneName=head}
   }
   n11 = Anim Dynamics - Bone: dyn_hair_a_01(ComponentPose=@n10, Alpha=0.6, ExternalForce="0,0,0") {
    angularDampingOverride=1
    angularSpringConstant=500
    bAngularSpring=true
    bOverrideAngularDamping=true
    bUseGravityOverride=true
    bUseSphericalLimits=true
    gravityScale=0.35
    simulationSpace=RootRelative
    alphaBoolBlend {blendOption=Linear}
    boundBone {boneName=dyn_hair_a_01}
    componentLinearAccScale=<1 1 1>
    componentLinearVelScale=<1 1 1>
    gravityOverride {z=-980}
    physicsBodyDefinitions [{
      collisionType=OuterSphere
      boundBone {boneName=dyn_hair_a_01}
      boxExtents=<10 10 10>
      constraintSetup {angularLimitsMax=<20 0 20>, angularLimitsMin=<-20 -30 -20>, angularTarget=<0 0 0>}
      localJointOffset=<3 0 0>
     }]
    relativeSpaceBone {boneName=head}
   }
   n12 = Anim Dynamics - Bone: dyn_hair_a_02(ComponentPose=@n11, Alpha=0.7, ExternalForce="0,0,0") {
    angularDampingOverride=1
    angularSpringConstant=500
    bAngularSpring=true
    bOverrideAngularDamping=true
    bUseGravityOverride=true
    bUseSphericalLimits=true
    gravityScale=0.35
    simulationSpace=RootRelative
    alphaBoolBlend {blendOption=Linear}
    boundBone {boneName=dyn_hair_a_02}
    componentLinearAccScale=<1 1 1>
    componentLinearVelScale=<1 1 1>
    gravityOverride {z=-980}
    physicsBodyDefinitions [{
      collisionType=OuterSphere
      boundBone {boneName=dyn_hair_a_02}
      boxExtents=<10 10 10>
      constraintSetup {angularLimitsMax=<20 0 20>, angularLimitsMin=<-20 -30 -20>, angularTarget=<0 0 0>}
      localJointOffset=<3 0 0>
     }]
    relativeSpaceBone {boneName=head}
    sphericalLimits [{limitRadius=11, drivingBone {boneName=spine_05}, sphereLocalOffset=<12.0796 -12.8205 12.2373>}]
   }
   n13 = Anim Dynamics - Bone: dyn_hair_b_01(ComponentPose=@n12, Alpha=0.5, ExternalForce="0,0,0") {
    angularDampingOverride=1
    angularSpringConstant=500
    bAngularSpring=true
    bOverrideAngularDamping=true
    bUseGravityOverride=true
    gravityScale=0.35
    simulationSpace=RootRelative
    alphaBoolBlend {blendOption=Linear}
    boundBone {boneName=dyn_hair_b_01}
    componentLinearAccScale=<1 1 1>
    componentLinearVelScale=<1 1 1>
    gravityOverride {z=-980}
    physicsBodyDefinitions [{
      boundBone {boneName=dyn_hair_b_01}
      boxExtents=<10 10 10>
      constraintSetup {angularLimitsMax=<20 20 20>, angularLimitsMin=<-20 -30 -20>, angularTarget=<0 0 0>}
      localJointOffset=<3 0 0>
     }]
    relativeSpaceBone {boneName=head}
   }
   n14 = Anim Dynamics - Bone: dyn_hair_b_02(ComponentPose=@n13, Alpha=0.6, ExternalForce="0,0,0") {
    angularDampingOverride=1
    angularSpringConstant=500
    bAngularSpring=true
    bOverrideAngularDamping=true
    bUseGravityOverride=true
    gravityScale=0.35
    simulationSpace=RootRelative
    alphaBoolBlend {blendOption=Linear}
    boundBone {boneName=dyn_hair_b_02}
    componentLinearAccScale=<1 1 1>
    componentLinearVelScale=<1 1 1>
    gravityOverride {z=-980}
    physicsBodyDefinitions [{
      boundBone {boneName=dyn_hair_b_02}
      boxExtents=<10 10 10>
      constraintSetup {angularLimitsMax=<20 20 20>, angularLimitsMin=<-20 -30 -20>, angularTarget=<0 0 0>}
      localJointOffset=<3 0 0>
     }]
    relativeSpaceBone {boneName=head}
   }
   n15 = Anim Dynamics - Bone: dyn_hair_e_02(ComponentPose=@n14, Alpha=0.7, ExternalForce="0,0,0") {
    angularDampingOverride=1
    angularSpringConstant=500
    bAngularSpring=true
    bOverrideAngularDamping=true
    bUseGravityOverride=true
    gravityScale=0.35
    simulationSpace=RootRelative
    alphaBoolBlend {blendOption=Linear}
    boundBone {boneName=dyn_hair_e_02}
    componentLinearAccScale=<1 1 1>
    componentLinearVelScale=<1 1 1>
    gravityOverride {z=-980}
    physicsBodyDefinitions [{
      boundBone {boneName=dyn_hair_e_02}
      boxExtents=<10 10 10>
      constraintSetup {angularLimitsMax=<20 5 20>, angularLimitsMin=<-20 0 -20>, angularTarget=<0 0 0>}
      localJointOffset=<3 0 0>
     }]
    relativeSpaceBone {boneName=head}
   }
   n16 = Anim Dynamics - Bone: dyn_hair_e_03(ComponentPose=@n15, Alpha=0.7, ExternalForce="0,0,0") {
    angularDampingOverride=1
    angularSpringConstant=500
    bAngularSpring=true
    bOverrideAngularDamping=true
    bUseGravityOverride=true
    bUseSphericalLimits=true
    gravityScale=0.35
    simulationSpace=RootRelative
    alphaBoolBlend {blendOption=Linear}
    boundBone {boneName=dyn_hair_e_03}
    componentLinearAccScale=<1 1 1>
    componentLinearVelScale=<1 1 1>
    gravityOverride {z=-980}
    physicsBodyDefinitions [{
      collisionType=OuterSphere
      boundBone {boneName=dyn_hair_e_03}
      boxExtents=<10 10 10>
      constraintSetup {angularLimitsMax=<20 5 20>, angularLimitsMin=<-20 0 -20>, angularTarget=<0 0 0>}
      localJointOffset=<3 0 0>
     }]
    relativeSpaceBone {boneName=head}
    sphericalLimits [{limitRadius=11, drivingBone {boneName=spine_05}, sphereLocalOffset=<12.0796 -9.9549 -15.7181>}]
   }
   Output Pose(Result=Component To Local(ComponentPose=@n16)) {layerGroup=None}
  ]}
 EventGraph {nodes [
   Event BlueprintUpdateAnimation()
   > n1 = Gravity Override State Function DEPRECATED(CharacterPartAnimInstance=Self-Reference(), GameTimeInSeconds=GetGameTimeInSeconds())
   > Set Gravity Override(Gravity Override=@n1)
   n2 = GetOwnerSkeletalMeshComponent()
   > Set head rotation left(head rotation left:double=MapRangeClamped(
     Value=GetDeltaTransformFromRefPose!(self=@n2, BoneName=head, BaseName=neck_01) -> (
      ReturnValue_Location:Vector
      ReturnValue_Rotation_Roll:float
      ReturnValue_Rotation_Pitch:float
      ReturnValue_Rotation_Yaw:float
      ReturnValue_Scale:Vector
     ).ReturnValue_Rotation_Roll
     InRangeA=0
     InRangeB=45
    )) -> (Output_Get:double)
   n3 = GetOwnerSkeletalMeshComponent()
   > Set head rotation right(head rotation right:double=MapRangeClamped(Value=GetDeltaTransformFromRefPose(self=@n3, BoneName=head, BaseName=neck_01).ReturnValue_Rotation_Roll, InRangeA=0, InRangeB=-45)) -> (Output_Get:double)
  ]}
]
variables [
 Gravity Override {type=Vector}
 head rotation left {type=number}
 head rotation right {type=number}
]
